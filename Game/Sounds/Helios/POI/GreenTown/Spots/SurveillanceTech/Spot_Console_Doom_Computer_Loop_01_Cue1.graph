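# Ambient spot loop for a computer console in GreenTown's SurveillanceTech area, routed to the indoor ambient sound class.
# A mixer blends a modulated communication hum loop with a Doom tech loop that an enveloper fades in.
# Two looping random layers add pitch-varied computer one-shots, one of them spaced by random delays.
# Overridden attenuation adds low-pass distance filtering, listener focus and occlusion.
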SoundCue /Game/Sounds/Helios/POI/GreenTown/Spots/SurveillanceTech/Spot_Console_Doom_Computer_Loop_01_Cue1
root=Mixer_5
nodes [
 {id=Delay_1, children [Modulator_2], properties {delayMax=1, delayMin=0.5}}
 {
  id=Enveloper_0
  children [WavePlayer_1]
  properties {pitchCurve {editorCurveData {keys [{time=0}]}}, volumeCurve {editorCurveData {keys [<at 0 0>,<at 1.5 1>]}}}
 }
 {id=Looping_0, children [Modulator_1]}
 {id=Looping_1, children [Delay_1]}
 {id=Mixer_5, children [Looping_0,Looping_1,Modulator_0,Enveloper_0], properties {inputVolume [0.4,0.2,3,0.4]}}
 {id=Modulator_0, children [WavePlayer_0], properties {pitchMin=1}}
 {id=Modulator_1, children [Random_0], properties {pitchMax=1.15, volumeMax=1, volumeMin=1}}
 {id=Modulator_2, children [Random_2], properties {pitchMax=1.15, volumeMax=1, volumeMin=1}}
 {id=Random_0, children [WavePlayer_5,WavePlayer_6,WavePlayer_7,WavePlayer_8], properties {weights [1,1,1,1]}}
 {id=Random_2, children [WavePlayer_4, WavePlayer_10], properties {weights [1,1]}}
 {
  id=WavePlayer_0
  properties {bLooping=true, soundWaveAssetPtr=/Game/Sounds/Helios/Environment_Actors/ConsoleTerminals/Spot_Console_Terminal_Communication_Hum_Loop_04}
 }
 {
  id=WavePlayer_1
  properties {bLooping=true, soundWaveAssetPtr=&/Spot_Doom_Tech_Loop_02}
 }
 {id=WavePlayer_10, properties {soundWaveAssetPtr=/Game/Sounds/Apollo/Biomes/GovernmentSecretComplex/SecretComplex_Computer_Oneshot_07}}
 {id=WavePlayer_4, properties {soundWaveAssetPtr=/Game/Sounds/Apollo/Biomes/GovernmentSecretComplex/SecretComplex_Computer_Oneshot_01}}
 {id=WavePlayer_5, properties {soundWaveAssetPtr=/Game/Sounds/Apollo/Biomes/GovernmentSecretComplex/SecretComplex_Computer_Oneshot_02}}
 {id=WavePlayer_6, properties {soundWaveAssetPtr=/Game/Sounds/Apollo/Biomes/GovernmentSecretComplex/SecretComplex_Computer_Oneshot_03}}
 {id=WavePlayer_7, properties {soundWaveAssetPtr=/Game/Sounds/Apollo/Biomes/GovernmentSecretComplex/SecretComplex_Computer_Oneshot_04}}
 {id=WavePlayer_8, properties {soundWaveAssetPtr=/Game/Sounds/Apollo/Biomes/GovernmentSecretComplex/SecretComplex_Computer_Oneshot_05}}
]
properties {
 bOverrideAttenuation=true
 maxDistance=1300
 soundClassObject=&Fort_Ambient_Indoors_Athena
 volumeMultiplier=0.551
 attenuationOverrides {
  bAttenuateWithLPF=true
  bEnableListenerFocus=true
  bEnableOcclusion=true
  falloffDistance=900
  focusAzimuth=40
  nonFocusAzimuth=120
  nonFocusVolumeAttenuation=0.5
  occlusionInterpolationTime=0.75
  occlusionLowPassFilterFrequency=2000
  occlusionTraceChannel=ECC_Camera
  occlusionVolumeAttenuation=0.5
 }
 concurrencyOverrides {maxCount=3, resolutionRule=PreventNew}
}
