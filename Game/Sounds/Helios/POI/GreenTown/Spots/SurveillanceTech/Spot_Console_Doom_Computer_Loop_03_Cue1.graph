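# Looping ambient spot Sound Cue for a "Doom" computer console in the GreenTown surveillance tech set, routed to an indoor ambience sound class.
# A Mixer layers three looping waves: a server-rack scan through a Modulator, a terminal hum pitched down by a Modulator, and a Doom tech loop.
# Envelopers fade in the whole mix, with a slower fade on the Doom layer.
# Overrides add box attenuation with occlusion and a small concurrency cap.
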SoundCue /Game/Sounds/Helios/POI/GreenTown/Spots/SurveillanceTech/Spot_Console_Doom_Computer_Loop_03_Cue1
root=Enveloper_0
nodes [
 {
  id=Enveloper_0
  children [Mixer_0]
  properties {pitchCurve {editorCurveData {keys [{time=0}]}}, volumeCurve {editorCurveData {keys [<at 0 0>,<at 1 1>]}}}
 }
 {
  id=Enveloper_1
  children [WavePlayer_2]
  properties {pitchCurve {editorCurveData {keys [{time=0}]}}, volumeCurve {editorCurveData {keys [<at 0 0>,<at 1.5 1>]}}}
 }
 {id=Mixer_0, children [Modulator_0,Modulator_1,Enveloper_1], properties {inputVolume [2,1,0.8]}}
 {id=Modulator_0, children [WavePlayer_1]}
 {id=Modulator_1, children [WavePlayer_0], properties {pitchMax=0.8, pitchMin=0.8}}
 {
  id=WavePlayer_0
  properties {bLooping=true, soundWaveAssetPtr=/Game/Sounds/Helios/Environment_Actors/ConsoleTerminals/Spot_Console_Terminal_Communication_Hum_Loop_02}
 }
 {
  id=WavePlayer_1
  properties {bLooping=true, soundWaveAssetPtr=/Game/Sounds/Helios/Environment_Actors/ServerRacks/Spot_Server_Stack_Scanning_Loop_03}
 }
 {
  id=WavePlayer_2
  properties {bLooping=true, soundWaveAssetPtr=&/Spot_Doom_Tech_Loop_04}
 }
]
properties {
 bOverrideAttenuation=true
 bOverrideConcurrency=true
 maxDistance=871.7567
 priority=0.25
 soundClassObject=&Fort_Ambient_Indoors_Athena
 volumeMultiplier=0.351
 attenuationOverrides {
  attenuationShape=Box
  falloffDistance=639.4926
  nonSpatializedRadiusMode=Direct2D
  nonSpatializedRadiusStart=1028.7994
  occlusionTraceChannel=ECC_Camera
  occlusionVolumeAttenuation=0.5
  attenuationShapeExtents=<213.7515 232.2641 -141.0018>
 }
 concurrencyOverrides {maxCount=3}
}
