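# Looping ambient spot cue for a tech console among the GreenTown surveillance-tech sounds, routed to the indoor ambient sound class.
# One looping wave passes through a mixer with an empty second input, then an enveloper that fades volume in over a short ramp.
# The root attenuation node overrides radio-console settings with listener focus, ducking out-of-focus sound.
# A concurrency override caps simultaneous instances and blocks new ones beyond the limit.
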SoundCue /Game/Sounds/Helios/POI/GreenTown/Spots/SurveillanceTech/Spot_Console_Doom_Tech_Loop_09_Cue1
root=Attenuation_0
nodes [
 {
  id=Attenuation_0
  children [Enveloper_0]
  properties {
   attenuationSettings=/Game/Sounds/Helios/Environment_Actors/Console_RadioCommunication/Spot_Console_RadioCommunication_Loop_Att
   bOverrideAttenuation=true
   attenuationOverrides {
    bEnableFocusInterpolation=true
    bEnableListenerFocus=true
    falloffDistance=1200
    focusReleaseInterpSpeed=0.9
    nonFocusVolumeAttenuation=0.4
    attenuationShapeExtents {x=300}
   }
  }
 }
 {
  id=Enveloper_0
  children [Mixer_0]
  properties {pitchCurve {editorCurveData {keys [{time=0}]}}, volumeCurve {editorCurveData {keys [<at 0 0>,<at 1.5 1>]}}}
 }
 {id=Mixer_0, children [WavePlayer_0,null], properties {inputVolume [1,0.9]}}
 {
  id=WavePlayer_0
  properties {bLooping=true, soundWaveAssetPtr=&/Spot_Doom_Tech_Loop_09}
 }
]
properties {
 bOverrideConcurrency=true
 maxDistance=1500
 soundClassObject=&Fort_Ambient_Indoors_Athena
 volumeMultiplier=0.501
 concurrencyOverrides {maxCount=4, resolutionRule=PreventNew, volumeScale=0.5}
}
