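# Looping ambient spot Sound Cue for a tech console, filed under the GreenTown POI's SurveillanceTech sounds.
# A single looping Wave Player feeds the Doom tech loop wave straight to the output at reduced volume.
# It borrows the radio-communication console's attenuation asset and uses the indoor ambient sound class.
# Overridden concurrency caps simultaneous instances, refusing new ones past the limit and applying volume scaling.
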SoundCue /Game/Sounds/Helios/POI/GreenTown/Spots/SurveillanceTech/Spot_Console_Doom_Tech_Loop_10_Cue1
root=WavePlayer_0
nodes [{
  id=WavePlayer_0
  properties {bLooping=true, soundWaveAssetPtr=&/Spot_Doom_Tech_Loop_10}
 }]
properties {
 attenuationSettings=/Game/Sounds/Helios/Environment_Actors/Console_RadioCommunication/Spot_Console_RadioCommunication_Loop_Att
 bOverrideConcurrency=true
 maxDistance=850
 soundClassObject=&Fort_Ambient_Indoors_Athena
 volumeMultiplier=0.501
 concurrencyOverrides {maxCount=4, resolutionRule=PreventNew, volumeScale=0.5}
}
