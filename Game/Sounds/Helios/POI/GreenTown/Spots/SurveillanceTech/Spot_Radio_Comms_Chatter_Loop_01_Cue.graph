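# Looping spot ambience cue of surveillance radio comms chatter for the GreenTown POI, in the indoor ambient sound class.
# A mixer layers a quiet looping glitchy radio bed under an endlessly looped, equally weighted random pick of seven chatter waves.
# An enveloper fades the mix in, then a root attenuation node applies radio console settings with overrides.
# Concurrency overrides cap simultaneous instances and block new ones beyond the limit.
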SoundCue /Game/Sounds/Helios/POI/GreenTown/Spots/SurveillanceTech/Spot_Radio_Comms_Chatter_Loop_01_Cue
root=Attenuation_0
nodes [
 {
  id=Attenuation_0
  children [Enveloper_0]
  properties {
   attenuationSettings=/Game/Sounds/Helios/Environment_Actors/Console_RadioCommunication/Spot_Console_RadioCommunication_Loop_Att
   bOverrideAttenuation=true
   attenuationOverrides {falloffDistance=1200, attenuationShapeExtents {x=250}}
  }
 }
 {
  id=Enveloper_0
  children [Mixer_0]
  properties {pitchCurve {editorCurveData {keys [{time=0}]}}, volumeCurve {editorCurveData {keys [<at 0 0>,<at 1.5 1>]}}}
 }
 {id=Looping_0, children [Random_3]}
 {id=Mixer_0, children [WavePlayer_1,Looping_0], properties {inputVolume [0.15,1]}}
 {
  id=Random_3
  children [WavePlayer_26, WavePlayer_27, WavePlayer_28, WavePlayer_29, WavePlayer_30, WavePlayer_31, WavePlayer_32]
  properties {weights [1,1,1,1,1,1,1]}
 }
 {id=WavePlayer_1, properties {bLooping=true, soundWaveAssetPtr=/Game/Sounds/Helios/POI/SocietyEstate/Spots/Spot_Radio_Comms_Glitchy_Loop}}
 {id=WavePlayer_26, properties {soundWaveAssetPtr=&/Spot_Radio_Surveillance_Chatter_02}}
 {id=WavePlayer_27, properties {soundWaveAssetPtr=&/Spot_Radio_Surveillance_Chatter_03}}
 {id=WavePlayer_28, properties {soundWaveAssetPtr=&/Spot_Radio_Surveillance_Chatter_04}}
 {id=WavePlayer_29, properties {soundWaveAssetPtr=&/Spot_Radio_Surveillance_Chatter_05}}
 {id=WavePlayer_30, properties {soundWaveAssetPtr=&/Spot_Radio_Surveillance_Chatter_06}}
 {id=WavePlayer_31, properties {soundWaveAssetPtr=&/Spot_Radio_Surveillance_Chatter_07}}
 {id=WavePlayer_32, properties {soundWaveAssetPtr=&/Spot_Radio_Surveillance_Chatter_08}}
]
properties {
 bOverrideConcurrency=true
 maxDistance=1450
 soundClassObject=&Fort_Ambient_Indoors_Athena
 volumeMultiplier=0.501
 concurrencyOverrides {maxCount=4, resolutionRule=PreventNew, volumeScale=0.5}
}
