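# Sound cue for a GreenTown surveillance-tech radio spot, playing endless comms chatter as indoor ambient sound.
# A mixer layers a quieter looping glitchy radio bed with a louder chatter branch that starts after a random delay.
# That branch loops a second random delay into an equally weighted pick of eight chatter clips, giving irregular gaps.
# It uses radio console attenuation and a concurrency override that blocks new instances past a small limit.
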SoundCue /Game/Sounds/Helios/POI/GreenTown/Spots/SurveillanceTech/Spot_Radio_Comms_Chatter_Loop_01_Cue1
root=Mixer_0
nodes [
 {id=Delay_0, children [Looping_0], properties {delayMax=2, delayMin=0.75}}
 {id=Delay_1, children [Random_3], properties {delayMax=2, delayMin=0.75}}
 {id=Looping_0, children [Delay_1]}
 {id=Mixer_0, children [WavePlayer_1,Delay_0], properties {inputVolume [0.2,0.75]}}
 {
  id=Random_3
  children [WavePlayer_25, WavePlayer_26, WavePlayer_27, WavePlayer_28, WavePlayer_29, WavePlayer_30, WavePlayer_31, WavePlayer_32]
  properties {weights [1,1,1,1,1,1,1,1]}
 }
 {id=WavePlayer_1, properties {bLooping=true, soundWaveAssetPtr=/Game/Sounds/Helios/POI/SocietyEstate/Spots/Spot_Radio_Comms_Glitchy_Loop}}
 {id=WavePlayer_25, properties {soundWaveAssetPtr=&/Spot_Radio_Surveillance_Chatter_01}}
 {id=WavePlayer_26, properties {soundWaveAssetPtr=&/Spot_Radio_Surveillance_Chatter_02}}
 {id=WavePlayer_27, properties {soundWaveAssetPtr=&/Spot_Radio_Surveillance_Chatter_03}}
 {id=WavePlayer_28, properties {soundWaveAssetPtr=&/Spot_Radio_Surveillance_Chatter_04}}
 {id=WavePlayer_29, properties {soundWaveAssetPtr=&/Spot_Radio_Surveillance_Chatter_05}}
 {id=WavePlayer_30, properties {soundWaveAssetPtr=&/Spot_Radio_Surveillance_Chatter_06}}
 {id=WavePlayer_31, properties {soundWaveAssetPtr=&/Spot_Radio_Surveillance_Chatter_07}}
 {id=WavePlayer_32, properties {soundWaveAssetPtr=&/Spot_Radio_Surveillance_Chatter_08}}
]
properties {
 attenuationSettings=/Game/Sounds/Helios/Environment_Actors/Console_RadioCommunication/Spot_Console_RadioCommunication_Loop_Att
 maxDistance=850
 soundClassObject=&Fort_Ambient_Indoors_Athena
 volumeMultiplier=0.251
 concurrencyOverrides {maxCount=4, resolutionRule=PreventNew, volumeScale=0.5}
}
