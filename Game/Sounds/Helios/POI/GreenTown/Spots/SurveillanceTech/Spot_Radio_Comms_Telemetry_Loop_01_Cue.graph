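# Looping indoor ambient cue for a radio comms telemetry spot among the GreenTown POI surveillance tech sounds.
# A mixer layers a constant looping terminal hum with a quieter looped chain that, after a random short delay, plays one of six telemetry clips with randomized pitch.
# An enveloper fades the mix in over the opening moments, then an overridden console radio attenuation spatializes it, with concurrency capped to prevent new instances.
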SoundCue /Game/Sounds/Helios/POI/GreenTown/Spots/SurveillanceTech/Spot_Radio_Comms_Telemetry_Loop_01_Cue
root=Attenuation_0
nodes [
 {
  id=Attenuation_0
  children [Enveloper_0]
  properties {
   attenuationSettings=/Game/Sounds/Helios/Environment_Actors/Console_RadioCommunication/Spot_Console_RadioCommunication_Loop_Att
   bOverrideAttenuation=true
   attenuationOverrides {falloffDistance=3000}
  }
 }
 {id=Delay_6, children [Random_1], properties {delayMax=4, delayMin=0.2}}
 {
  id=Enveloper_0
  children [Mixer_0]
  properties {pitchCurve {editorCurveData {keys [{time=0}]}}, volumeCurve {editorCurveData {keys [<at 0 0>,<at 1.5 1>]}}}
 }
 {id=Looping_0, children [Modulator_0]}
 {id=Mixer_0, children [WavePlayer_7,Looping_0], properties {inputVolume [1,0.5]}}
 {id=Modulator_0, children [Delay_6], properties {pitchMax=1.2, pitchMin=0.85}}
 {
  id=Random_1
  children [WavePlayer_8, WavePlayer_9, WavePlayer_10, WavePlayer_11, WavePlayer_12, WavePlayer_13]
  properties {weights [1,1,1,1,1,1]}
 }
 {id=WavePlayer_10, properties {soundWaveAssetPtr=&/Spot_Radio_Telemetry_A_4}}
 {id=WavePlayer_11, properties {soundWaveAssetPtr=&/Spot_Radio_Telemetry_A_3}}
 {id=WavePlayer_12, properties {soundWaveAssetPtr=&/Spot_Radio_Telemetry_A_2}}
 {id=WavePlayer_13, properties {soundWaveAssetPtr=&/Spot_Radio_Telemetry_A_1}}
 {
  id=WavePlayer_7
  properties {bLooping=true, soundWaveAssetPtr=/Game/Sounds/Helios/Environment_Actors/ConsoleTerminals/Spot_Console_Terminal_Communication_Hum_Loop_04}
 }
 {id=WavePlayer_8, properties {soundWaveAssetPtr=&/Spot_Radio_Telemetry_A_6}}
 {id=WavePlayer_9, properties {soundWaveAssetPtr=&/Spot_Radio_Telemetry_A_5}}
]
properties {
 bOverrideConcurrency=true
 maxDistance=3400
 soundClassObject=&Fort_Ambient_Indoors_Athena
 volumeMultiplier=0.501
 concurrencyOverrides {maxCount=4, resolutionRule=PreventNew, volumeScale=0.5}
}
