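# Ambient SoundCue for a looping radio telemetry spot sound in the GreenTown surveillance tech POI, using console radio attenuation and an indoor ambient sound class.
# A fade-in envelope wraps a three-input mixer: a quiet, pitched-up looping tech bed plus two looping layers of telemetry chatter.
# Each chatter layer waits a random delay, then plays one of the same eight telemetry waves through its own modulator, giving irregular, overlapping transmissions.
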SoundCue /Game/Sounds/Helios/POI/GreenTown/Spots/SurveillanceTech/Spot_Radio_Comms_Telemetry_Loop_02_Cue1
root=Attenuation_0
nodes [
 {
  id=Attenuation_0
  children [Enveloper_0]
  properties {attenuationSettings=/Game/Sounds/Helios/Environment_Actors/Console_RadioCommunication/Spot_Console_RadioCommunication_Loop_Att}
 }
 {id=Delay_0, children [Random_2], properties {delayMax=4, delayMin=0.2}}
 {id=Delay_6, children [Random_0], properties {delayMax=4, delayMin=0.2}}
 {
  id=Enveloper_0
  children [Mixer_0]
  properties {pitchCurve {editorCurveData {keys [{time=0}]}}, volumeCurve {editorCurveData {keys [<at 0 0>,<at 1.5 1>]}}}
 }
 {id=Looping_0, children [Modulator_0]}
 {id=Looping_1, children [Modulator_2]}
 {id=Mixer_0, children [Modulator_1,Looping_0,Looping_1], properties {inputVolume [0.25,0.9,0.4]}}
 {id=Modulator_0, children [Delay_6], properties {pitchMax=1.2}}
 {id=Modulator_1, children [WavePlayer_7], properties {pitchMax=1.5, pitchMin=1.4}}
 {id=Modulator_2, children [Delay_0]}
 {
  id=Random_0
  children [WavePlayer_0, WavePlayer_1, WavePlayer_2, WavePlayer_3, WavePlayer_4, WavePlayer_5, WavePlayer_6, WavePlayer_14]
  properties {weights [1,1,1,1,1,1,1,1]}
 }
 {
  id=Random_2
  children [WavePlayer_16, WavePlayer_17, WavePlayer_18, WavePlayer_19, WavePlayer_20, WavePlayer_21, WavePlayer_22, WavePlayer_23]
  properties {weights [1,1,1,1,1,1,1,1]}
 }
 {id=WavePlayer_0, properties {soundWaveAssetPtr=&/Spot_Radio_Telemetry_B_1}}
 {id=WavePlayer_1, properties {soundWaveAssetPtr=&/Spot_Radio_Telemetry_B_2}}
 {id=WavePlayer_14, properties {soundWaveAssetPtr=&/Spot_Radio_Telemetry_B_8}}
 {id=WavePlayer_16, properties {soundWaveAssetPtr=&Spot_Radio_Telemetry_B_1}}
 {id=WavePlayer_17, properties {soundWaveAssetPtr=&Spot_Radio_Telemetry_B_2}}
 {id=WavePlayer_18, properties {soundWaveAssetPtr=&/Spot_Radio_Telemetry_B_3}}
 {id=WavePlayer_19, properties {soundWaveAssetPtr=&/Spot_Radio_Telemetry_B_4}}
 {id=WavePlayer_2, properties {soundWaveAssetPtr=&Spot_Radio_Telemetry_B_3}}
 {id=WavePlayer_20, properties {soundWaveAssetPtr=&/Spot_Radio_Telemetry_B_5}}
 {id=WavePlayer_21, properties {soundWaveAssetPtr=&/Spot_Radio_Telemetry_B_6}}
 {id=WavePlayer_22, properties {soundWaveAssetPtr=&/Spot_Radio_Telemetry_B_7}}
 {id=WavePlayer_23, properties {soundWaveAssetPtr=&Spot_Radio_Telemetry_B_8}}
 {id=WavePlayer_3, properties {soundWaveAssetPtr=&Spot_Radio_Telemetry_B_4}}
 {id=WavePlayer_4, properties {soundWaveAssetPtr=&Spot_Radio_Telemetry_B_5}}
 {id=WavePlayer_5, properties {soundWaveAssetPtr=&Spot_Radio_Telemetry_B_6}}
 {id=WavePlayer_6, properties {soundWaveAssetPtr=&Spot_Radio_Telemetry_B_7}}
 {
  id=WavePlayer_7
  properties {bLooping=true, soundWaveAssetPtr=&/Spot_Doom_Tech_Loop_03}
 }
]
properties {
 bOverrideAttenuation=true
 maxDistance=4000
 soundClassObject=&Fort_Ambient_Indoors_Athena
 volumeMultiplier=0.551
 concurrencyOverrides {volumeScale=0.5}
}
