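# Spot ambient sound cue for a radio comms telemetry console in the GreenTown surveillance tech area.
# Two looping layers play randomly delayed, pitch-varied picks from the same seven telemetry clips at different mix levels, over a raised-pitch looping console bed.
# An Enveloper fades the mix in, and an attenuation node with overridden falloff and extents spatializes it.
# A concurrency override caps instances and blocks new ones at the limit.
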SoundCue /Game/Sounds/Helios/POI/GreenTown/Spots/SurveillanceTech/Spot_Radio_Comms_Telemetry_Loop_03b_Cue
root=Attenuation_0
nodes [
 {
  id=Attenuation_0
  children [Enveloper_0]
  properties {
   attenuationSettings=/Game/Sounds/Helios/Environment_Actors/Console_RadioCommunication/Spot_Console_RadioCommunication_Loop_Att
   bOverrideAttenuation=true
   attenuationOverrides {falloffDistance=850, attenuationShapeExtents {x=200}}
  }
 }
 {id=Delay_0, children [Random_3], properties {delayMax=4, delayMin=0.2}}
 {id=Delay_6, children [Random_1], properties {delayMax=4, delayMin=0.2}}
 {
  id=Enveloper_0
  children [Mixer_0]
  properties {pitchCurve {editorCurveData {keys [{time=0}]}}, volumeCurve {editorCurveData {keys [<at 0 0>,<at 1.5 1>]}}}
 }
 {id=Looping_0, children [Modulator_0]}
 {id=Looping_1, children [Modulator_2]}
 {id=Mixer_0, children [Looping_0,Looping_1,Modulator_3], properties {inputVolume [0.45,0.1,1]}}
 {id=Modulator_0, children [Delay_6], properties {pitchMax=1.1}}
 {id=Modulator_2, children [Delay_0], properties {pitchMax=1.2, pitchMin=0.85}}
 {id=Modulator_3, children [WavePlayer_0], properties {pitchMax=1.5, pitchMin=1.45}}
 {
  id=Random_1
  children [WavePlayer_7, WavePlayer_8, WavePlayer_9, WavePlayer_10, WavePlayer_11, WavePlayer_12, WavePlayer_13]
  properties {weights [1,1,1,1,1,1,1]}
 }
 {
  id=Random_3
  children [WavePlayer_24, WavePlayer_25, WavePlayer_26, WavePlayer_27, WavePlayer_28, WavePlayer_29, WavePlayer_30]
  properties {weights [1,1,1,1,1,1,1]}
 }
 {id=WavePlayer_0, properties {bLooping=true, soundWaveAssetPtr=/Game/Athena/Sounds/Props/Tinfoil_Drill/Tinfoil_Console_Loop_01}}
 {id=WavePlayer_10, properties {soundWaveAssetPtr=&/Spot_Radio_Telemetry_C_4}}
 {id=WavePlayer_11, properties {soundWaveAssetPtr=&/Spot_Radio_Telemetry_C_5}}
 {id=WavePlayer_12, properties {soundWaveAssetPtr=&/Spot_Radio_Telemetry_C_6}}
 {id=WavePlayer_13, properties {soundWaveAssetPtr=&/Spot_Radio_Telemetry_C_7}}
 {id=WavePlayer_24, properties {soundWaveAssetPtr=&/Spot_Radio_Telemetry_C_1}}
 {id=WavePlayer_25, properties {soundWaveAssetPtr=&/Spot_Radio_Telemetry_C_2}}
 {id=WavePlayer_26, properties {soundWaveAssetPtr=&/Spot_Radio_Telemetry_C_3}}
 {id=WavePlayer_27, properties {soundWaveAssetPtr=&Spot_Radio_Telemetry_C_4}}
 {id=WavePlayer_28, properties {soundWaveAssetPtr=&Spot_Radio_Telemetry_C_5}}
 {id=WavePlayer_29, properties {soundWaveAssetPtr=&Spot_Radio_Telemetry_C_6}}
 {id=WavePlayer_30, properties {soundWaveAssetPtr=&Spot_Radio_Telemetry_C_7}}
 {id=WavePlayer_7, properties {soundWaveAssetPtr=&Spot_Radio_Telemetry_C_1}}
 {id=WavePlayer_8, properties {soundWaveAssetPtr=&Spot_Radio_Telemetry_C_2}}
 {id=WavePlayer_9, properties {soundWaveAssetPtr=&Spot_Radio_Telemetry_C_3}}
]
properties {
 bOverrideConcurrency=true
 maxDistance=1050
 soundClassObject=&Fort_Ambient_Indoors_Athena
 volumeMultiplier=0.501
 concurrencyOverrides {maxCount=4, resolutionRule=PreventNew, volumeScale=0.5}
}
