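# Sound cue for a looping radio comms and telemetry ambient spot in the GreenTown surveillance-tech POI.
# It mixes a pitched-up, looping console bed with a quieter layer that repeatedly plays one of eight random telemetry one-shots after random delays, with pitch modulation.
# An enveloper fades the mix in before shared console radio attenuation, and a concurrency override blocks new instances past a small cap.
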SoundCue /Game/Sounds/Helios/POI/GreenTown/Spots/SurveillanceTech/Spot_Radio_Comms_Telemetry_Loop_05b_Cue
root=Attenuation_0
nodes [
 {
  id=Attenuation_0
  children [Enveloper_0]
  properties {attenuationSettings=/Game/Sounds/Helios/Environment_Actors/Console_RadioCommunication/Spot_Console_RadioCommunication_Loop_Att}
 }
 {id=Delay_6, children [Random_0], properties {delayMax=4, delayMin=0.2}}
 {
  id=Enveloper_0
  children [Mixer_0]
  properties {pitchCurve {editorCurveData {keys [{time=0}]}}, volumeCurve {editorCurveData {keys [<at 0 0>,<at 1.5 1>]}}}
 }
 {id=Looping_0, children [Modulator_0]}
 {id=Mixer_0, children [Modulator_1,Looping_0], properties {inputVolume [1,0.2]}}
 {id=Modulator_0, children [Delay_6], properties {pitchMax=1.1}}
 {id=Modulator_1, children [WavePlayer_8], properties {pitchMax=1.4, pitchMin=1.2}}
 {
  id=Random_0
  children [WavePlayer_0, WavePlayer_1, WavePlayer_2, WavePlayer_3, WavePlayer_4, WavePlayer_5, WavePlayer_6, WavePlayer_14]
  properties {weights [1,1,1,1,1,1,1,1]}
 }
 {id=WavePlayer_0, properties {soundWaveAssetPtr=&/Spot_Radio_Telemetry_D_1}}
 {id=WavePlayer_1, properties {soundWaveAssetPtr=&/Spot_Radio_Telemetry_D_2}}
 {id=WavePlayer_14, properties {soundWaveAssetPtr=&/Spot_Radio_Telemetry_D_8}}
 {id=WavePlayer_2, properties {soundWaveAssetPtr=&/Spot_Radio_Telemetry_D_3}}
 {id=WavePlayer_3, properties {soundWaveAssetPtr=&/Spot_Radio_Telemetry_D_4}}
 {id=WavePlayer_4, properties {soundWaveAssetPtr=&/Spot_Radio_Telemetry_D_5}}
 {id=WavePlayer_5, properties {soundWaveAssetPtr=&/Spot_Radio_Telemetry_D_6}}
 {id=WavePlayer_6, properties {soundWaveAssetPtr=&/Spot_Radio_Telemetry_D_7}}
 {id=WavePlayer_8, properties {bLooping=true, soundWaveAssetPtr=/Game/Athena/Sounds/Props/Tinfoil_Drill/Tinfoil_Console_Loop_01}}
]
properties {
 attenuationSettings=&Spot_Console_RadioCommunication_Loop_Att
 maxDistance=850
 soundClassObject=&Fort_Ambient_Indoors_Athena
 volumeMultiplier=0.501
 concurrencyOverrides {maxCount=4, resolutionRule=PreventNew, volumeScale=0.5}
}
